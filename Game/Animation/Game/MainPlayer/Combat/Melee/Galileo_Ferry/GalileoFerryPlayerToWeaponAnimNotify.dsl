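; Anim notify on Galileo Ferry player animations that signals the one-handed Galileo Ferry pickaxe to fold back.
; On notify, it casts the mesh owner to a Fort player pawn and takes its current weapon as the Galileo Ferry pickaxe.
; If that weapon mesh is valid, it sets ShouldFoldBackWeapon on the weapon's anim blueprint, bridging player animation to weapon animation.
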(fn Received_Notify
  (params (MeshComp :type Skeletal%20Mesh%20Component%20Object%20Reference) (Animation :type Anim%20Sequence%20Base%20Object%20Reference) (EventReference :type Anim%20Notify%20Event%20Reference%20Structure%20%28by%20ref%29))
  (bind _asfort_player_pawn (Utilities|Casting|CastToFortPlayerPawn (Components|GetOwner MeshComp)))
  (bind _asb_athena_pickaxe_galileo_ferry_1h (Utilities|Casting|CastToB_Athena_Pickaxe_Galileo_Ferry1H (Variables|/Script/FortniteGame.FortPawn|GetCurrentWeapon _asfort_player_pawn)))
  (Utilities|IsValid (Variables|/Script/FortniteGame.FortWeapon|GetWeaponMesh _asb_athena_pickaxe_galileo_ferry_1h)
    (:Is%20Valid
      (bind _asgalileo_ferry_axe_anim_bp (Utilities|Casting|CastToGalileo_Ferry_Axe_AnimBP (Components|SkeletalMesh|GetAnimInstance (Variables|/Script/FortniteGame.FortWeapon|GetWeaponMesh _asb_athena_pickaxe_galileo_ferry_1h))))
      (Variables|/Script/FortniteGame.AnimInstance_GalileoFerryAxe|SetShouldFoldBackWeapon true _asgalileo_ferry_axe_anim_bp)
      (return true))))
