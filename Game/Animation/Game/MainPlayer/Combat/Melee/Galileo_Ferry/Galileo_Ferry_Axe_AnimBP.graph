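# Animation Blueprint for the Galileo Ferry pickaxe weapon skeleton, switching the weapon between its base pose and a fold animation.
# A bool blend on ShouldFoldBackWeapon, not declared in this asset, plays the non-looping fold when true and the base pose otherwise, resetting the child on activation.
# The result passes through DefaultSlot so montages can override it, and the event graph is empty.
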AnimBlueprint /Game/Animation/Game/MainPlayer/Combat/Melee/Galileo_Ferry/Galileo_Ferry_Axe_AnimBP
target_skeleton=/Game/Weapons/FORT_Melee/Pickaxe_Galileo_Ferry/Meshes/Galileo_Ferry_Axe_Skeleton
graphs [
 AnimGraph {nodes [
   n1 = Sequence Player 'GalileoFerry_WeaponFold_Weapon'() {bLoopAnimation=false, sequence=&/GalileoFerry_WeaponFold_Weapon}
   n2 = Sequence Player 'GalileoFerry_BasePose_Weapon'() {sequence=&/GalileoFerry_BasePose_Weapon}
   n3 = Get ShouldFoldBackWeapon() -> (ShouldFoldBackWeapon:bool)
   Output Pose(Result=Slot 'DefaultSlot'(Source=Blend Poses by bool(BlendPose_0=@n1, BlendPose_1=@n2, bActiveValue=@n3) {blendType=Linear, childUpateMode=ResetChildOnActivate})) {layerGroup=None}
  ]}
 EventGraph {links [], nodes []}
]
variables []
