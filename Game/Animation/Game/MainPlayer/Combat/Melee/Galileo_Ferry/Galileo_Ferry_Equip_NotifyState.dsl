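; Anim notify state for the Galileo Ferry pickaxe, named for its equip animation, that shows the weapon's idle particle effect for the notify's duration.
; On begin, it casts the mesh's anim instance to the axe AnimBP, sets its desired idle particle visibility on and activates the owning pickaxe's idle particle component.
; On end, it clears that flag and deactivates the same component.
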(fn Received_NotifyBegin
  (params (MeshComp :type Skeletal%20Mesh%20Component%20Object%20Reference) (Animation :type Anim%20Sequence%20Base%20Object%20Reference) (TotalDuration :type Float%20%28single-precision%29) (EventReference :type Anim%20Notify%20Event%20Reference%20Structure%20%28by%20ref%29))
  (bind _asgalileo_ferry_axe_anim_bp (Utilities|Casting|CastToGalileo_Ferry_Axe_AnimBP (Components|SkeletalMesh|GetAnimInstance MeshComp)))
  (bind _asfort_weapon_pickaxe_athena (Utilities|Casting|CastToFortWeaponPickaxeAthena (Animation|GetOwningActor _asgalileo_ferry_axe_anim_bp)))
  (Variables|/Script/FortniteGame.AnimInstance_GalileoFerryAxe|SetbDesiredIdleParticleVisibility true _asgalileo_ferry_axe_anim_bp)
  (Components|Activation|Activate (Class|FortWeaponPickaxeAthena|GetIdlePSC _asfort_weapon_pickaxe_athena) false)
  (return true))

(fn Received_NotifyEnd
  (params (MeshComp :type Skeletal%20Mesh%20Component%20Object%20Reference) (Animation :type Anim%20Sequence%20Base%20Object%20Reference) (EventReference :type Anim%20Notify%20Event%20Reference%20Structure%20%28by%20ref%29))
  (bind _asgalileo_ferry_axe_anim_bp (Utilities|Casting|CastToGalileo_Ferry_Axe_AnimBP (Components|SkeletalMesh|GetAnimInstance MeshComp)))
  (bind _asfort_weapon_pickaxe_athena (Utilities|Casting|CastToFortWeaponPickaxeAthena (Animation|GetOwningActor _asgalileo_ferry_axe_anim_bp)))
  (Variables|/Script/FortniteGame.AnimInstance_GalileoFerryAxe|SetbDesiredIdleParticleVisibility false _asgalileo_ferry_axe_anim_bp)
  (Components|Activation|Deactivate (Class|FortWeaponPickaxeAthena|GetIdlePSC _asfort_weapon_pickaxe_athena))
  (return true))
